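# Physics asset for the PromDutch biped "Invis" mesh in CosmeticCompanions, giving that mesh a minimal collision and bounds shape.
# It holds a single body on the pelvis bone: one wide, short, nearly spherical capsule, rotated and slightly offset from the bone.
# No other bones have bodies and there are no constraints, so it defines no articulated ragdoll.
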PhysicsAsset /CosmeticCompanions/Assets/Biped/PromDutch/Meshes/Invis/PromDutch_Biped_Invis_PhysicsAsset
bodies {pelvis [{shapeType=Capsule, length=15.5691, radius=31.141, center=<11.3606 1.1375 0>, rotation=<rot 90 0 0>}]}
constraints []
